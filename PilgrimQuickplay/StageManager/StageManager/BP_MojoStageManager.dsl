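; Stage manager Blueprint for Pilgrim Quickplay that drives stage show sequences and colors from gameplay states and MIDI text events.
; At cosmetic BeginPlay it waits for the event router, then registers with the music playspace's StageShowManager component.
; Each event looks up a sequence pool to play and color settings from palette data assets.
; Colors cycle per beat, bar or timer, updating MPC_Sparks and the managed FxManagers.
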(event EventBeginPlay
  (|CallParentFunction)
  (Networking|CanExecuteCosmeticEvents
    (:True
      ; Register Stage Manager on Stage Show Manager Component, if present
      (bind _eventrouter (Pilgrim|PilgrimWaitForEventRouter 0)
        (:OnEventRouterAvailable
          (bind _output_get (Variables|Default|SetEventRouter _eventrouter))
          (bind _assparks_music_playspace (Utilities|Casting|CastToSparksMusicPlayspace (Components|GetOwner _output_get)))
          (Utilities|IsValid (Actor|GetComponentbyClass _assparks_music_playspace "/PilgrimQuickplay/StageManager/StageShowManager/BP_PilgrimQuickplay_StageShowManager.BP_PilgrimQuickplay_StageShowManager_C")
            (:Is%20Valid
              (Class|BPPilgrimQuickplayStageShowManager|InitializeStageManager (Actor|GetComponentbyClass _assparks_music_playspace "/PilgrimQuickplay/StageManager/StageShowManager/BP_PilgrimQuickplay_StageShowManager.BP_PilgrimQuickplay_StageShowManager_C") self))))))))

(fn ApplyColorSettings
  (params (ColorSettings :type Stage%20Manager%20Event%20Colors%20Structure))
  (bind (_synctype _frequency) (Utilities|Struct|BreakStageManagerEventColors (Variables|ColorShift|GetCurrentColorSettings)))
  ;[1 1 1] Setting global values and resetting cycle variables
  (bind _output_get (Variables|ColorShift|SetCurrentColorSettings ColorSettings))
  (Variables|ColorShift|SetColorSettingsIndex 0)
  (Variables|ColorShift|SetColorBeatCounter 0)
  ;[1 1 1] Update colors to first index of subpalette array
  (ColorShift|UpdateStageColors)
  ;[1 1 1] Reset event bindings
  (Default|UnbindEventfromOnBeatCalled (EventDispatchers|CreateEvent :!event CycleEventColorPalettePerBeat))
  (Utilities|Time|ClearandInvalidateTimerbyHandle (Variables|ColorShift|GetColorTimerHandle))
  ;[1 1 1] If we have more that one sub palette, it's a color cycle
  (if (> (Utilities|Array|Length (.EventColorPalettes (Utilities|Struct|BreakStageManagerEventColors _output_get))) 1)
    (if (Utilities|Enum|NotEqual(Enum) (Utilities|Struct|BreakStageManagerEventColors _output_get) "None")
      ;[1 1 1] Music driven cycle
      (Default|BindEventtoOnBeatCalled (EventDispatchers|CreateEvent :!event CycleEventColorPalettePerBeat))
      (else
        ;[1 1 1] Timer cycle
        (bind _returnvalue (Utilities|Time|SetTimerbyEvent (EventDispatchers|CreateEvent :!event CycleEventColor) (.Frequency (Utilities|Struct|BreakStageManagerEventColors _output_get)) true true))
        (Variables|ColorShift|SetColorTimerHandle _returnvalue)))
    ;[1 0.7249 0.4104] Debug Log
    (if (Variables|Default|GetEnableDebugLogs)
      (Development|PrintText (Utilities|Text|FormatText "NSLOCTEXT(\"[DC252DB94F2F9B61224EC987A5DD07B9]\", \"EA8C7B3A44969600C4999483A5D0C34E\", \"[StageManager] Color Cycle: Every {value} {type}\")" _frequency (Utilities|String|EnumtoString _synctype)) true false "(R=0.000000,G=0.660000,B=1.000000,A=1.000000)" 6e+01 "CPCurrentSettings"))
    (elif (Variables|Default|GetEnableDebugLogs)
      ;[1 0.6755 0.4177] Debug Log
      (Development|PrintText "INVTEXT(\"H[StageManager] Color Cycle: Static\")" true false "(R=0.000000,G=0.660000,B=1.000000,A=1.000000)" 6e+01 "CPCurrentSettings"))))

(fn CycleEventColor
  (bind (_synctype _frequency _eventcolorpalettes) (Utilities|Struct|BreakStageManagerEventColors (Variables|ColorShift|GetCurrentColorSettings)))
  ;[1 1 1] Advance to the next subpalette. Loops to first when reaching the array's end
  (Math|Integer|IncrementInt (Variables|ColorShift|GetColorSettingsIndex)
    (:%20%20
      (if (not (< (Variables|ColorShift|GetColorSettingsIndex) (Utilities|Array|Length _eventcolorpalettes)))
        (Variables|ColorShift|SetColorSettingsIndex 0))
      (ColorShift|UpdateStageColors))))

(fn CycleEventColorPalettePerBeat
  (params (BeatInBar :type Integer) (MeasureNumber :type Integer) (IsBeating :type Boolean))
  (bind _currentcolorsettings (Variables|ColorShift|GetCurrentColorSettings))
  (bind (_synctype _frequency) (Utilities|Struct|BreakStageManagerEventColors _currentcolorsettings))
  ;[1 1 1] If on beat mode, count per beat. If measure, count per measure. Only cycle to next color when reached the set frequency (every X beats, etc..)
  (switch Utilities|FlowControl|Switch|SwitchonEMidiClockSubdivisionQuantization _synctype
    (:Bar
      (if (== BeatInBar 1)
        (Math|Integer|IncrementInt (Variables|ColorShift|GetColorBeatCounter)
          (:%20%20
            _))))
    (:Beat
      (Math|Integer|IncrementInt (Variables|ColorShift|GetColorBeatCounter)
        (:%20%20
          _))))
  (if (== (Variables|ColorShift|GetColorBeatCounter) (Math|Float|TruncatetoInteger64 _frequency))
    ;[1 1 1] Reset beat counter and cycle to next sub palette
    (Variables|ColorShift|SetColorBeatCounter 0)
    (ColorShift|CycleEventColor)))

(fn ForceGameplayEventColorSettings
  (params (Event :type EPilgrimQuickplayState%20Enum :default "Invalid"))
  (if (Utilities|Enum|NotEqual(Enum) Event "Invalid")
    (ColorShift|SetCurrentColorPaletteasGameplay)
    (ColorShift|UpdateColorsbyGameplayEvent Event)
    (if (Variables|Default|GetEnableDebugLogs)
      (Development|PrintString (Utilities|String|Append "[StageManager] Forced Gameplay color settings: " (Utilities|String|EnumtoString Event)) true false "(R=1.000000,G=0.777671,B=0.000000,A=1.000000)" 5.0 "StageManagerColorForce"))))

(fn ForceMIDIEventColorSettings
  (params (Event :type ESparksMidiTextEvents%20Enum))
  (if (Utilities|Enum|NotEqual(Enum) Event "None")
    (ColorShift|SetCurrentColorPaletteasMIDI)
    (ColorShift|UpdateColorsbyMIDIEvent Event)
    (if (Variables|Default|GetEnableDebugLogs)
      (Development|PrintString (Utilities|String|Append "[StageManager] Forced MIDI color settings: " (Utilities|String|EnumtoString Event)) true false "(R=1.000000,G=0.777671,B=0.000000,A=1.000000)" 5.0 "StageManagerColorForce"))))

(fn GetGameplayEventColorSettings
  (params (State :type EPilgrimQuickplayState%20Enum%20%28by%20ref%29 :default "Invalid"))
  (bind _currentgameplayeventcp (Variables|ColorShift|GetCurrentGameplayEventCP))
  (bind _eventmap (Variables|/PilgrimQuickplay/StageManager/StageManager/ColorPalette/BP_MojoStageColorPaletteGameplay_DataAsset.BP_MojoStageColorPaletteGameplay_DataAsset_C|GetEventMap _currentgameplayeventcp))
  (bind (_value _returnvalue) (Utilities|Map|Find _eventmap State))
  (Utilities|IsValid _currentgameplayeventcp
    (:Is%20Valid
      (if _returnvalue
        (return _value true)
        (elif (Variables|Default|GetEnableDebugLogs)
          (Development|PrintString (Utilities|String|Append "[StageManager] No color settings found for this gameplay event: " (Utilities|String|EnumtoString State)))
          (return 0 false))))
    (:Is%20Not%20Valid
      (Fortnite|FortShippingLog self "[StageManager] Gameplay event color palette is invalid. Color palette won't be updated." true))))

(fn GetMIDIEventColorSettings
  (params (Event :type ESparksMidiTextEvents%20Enum%20%28by%20ref%29))
  (bind _currentmidieventscp (Variables|ColorShift|GetCurrentMidiEventsCP))
  (bind _eventmap (Variables|/PilgrimQuickplay/StageManager/StageManager/ColorPalette/BP_MojoStageColorPaletteMidi_DataAsset.BP_MojoStageColorPaletteMidi_DataAsset_C|GetEventMap _currentmidieventscp))
  (bind (_value _returnvalue) (Utilities|Map|Find _eventmap Event))
  (Utilities|IsValid _currentmidieventscp
    (:Is%20Valid
      (if _returnvalue
        (return _value true)
        (else
          _)))
    (:Is%20Not%20Valid
      _))
  (Fortnite|FortShippingLog self "[StageManager] MIDI event color palette returned invalid. Color palette won't be updated." true)
  (return 0 false))

(fn GetSequencePoolsFromMaps
  (Utilities|Array|Clear (LocalVariables|GetPools))
  (bind _values (Utilities|Map|Values (Variables|/Script/SparksStageShowRuntime.StageShowGameplayEventMap|GetEventMap (Variables|Default|GetGameplayEventMap))))
  (for _array_element _values
    (Utilities|Array|AddUnique (LocalVariables|GetPools) _array_element))
  (bind _values_2 (Utilities|Map|Values (Variables|/Script/SparksStageShowRuntime.StageShowMIDIEventMap|GetEventMap (Variables|Default|GetMIDIEventMap))))
  (for _array_element_2 _values_2
    (Utilities|Array|AddUnique (LocalVariables|GetPools) _array_element_2))
  (return (LocalVariables|GetPools)))

(fn PlayGameplayStateEvent
  (params (State :type EPilgrimQuickplayState%20Enum :default "Invalid"))
  (bind _output_get (Variables|Default|SetCurrentGameState State))
  (Default|CallOnEventCalled (Utilities|Name|EnumtoName _output_get))
  (if (Variables|Default|GetEnableDebugLogs)
    (Development|PrintText (Utilities|Text|FormatText "NSLOCTEXT(\"[DC252DB94F2F9B61224EC987A5DD07B9]\", \"3BBBF24240E5D600DBABAFB3C6F8CB73\", \"[StageManager]: Current gameplay event: {event}\")" (Utilities|String|EnumtoString State)) true true "(R=0.000000,G=0.660000,B=1.000000,A=1.000000)" 3e+01 "StageManagerGPEvent"))
  (Utilities|FlowControl|Sequence
    (:then_0
      ;[1 1 1] Update Color Palette Settings
      (ColorShift|UpdateColorsbyGameplayEvent (Variables|Default|GetCurrentGameState)))
    (:then_1
      ;[1 1 1] Play sequence
      (Sequences|PlayQuickplayEventSequence (Variables|/Script/SparksStageShowRuntime.StageShowGameplayEventMap|GetEventMap (Variables|Default|GetGameplayEventMap)) (Variables|Default|GetCurrentGameState)))))

(fn PlayMIDIEvent
  (params (Event :type ESparksMidiTextEvents%20Enum))
  (Variables|Default|SetCurrentMIDIEvent Event)
  (Utilities|FlowControl|Sequence
    (:then_0
      ;[1 1 1] Update Color Palette settings
      (ColorShift|UpdateColorsbyMIDIEvent (Variables|Default|GetCurrentMIDIEvent)))
    (:then_1
      ;[1 1 1] Get Random Sequence to play
      (Sequences|PlayMidiEventSequence (Variables|/Script/SparksStageShowRuntime.StageShowMIDIEventMap|GetEventMap (Variables|Default|GetMIDIEventMap)) (Variables|Default|GetCurrentMIDIEvent))
      (return))))

(fn PlayMidiEventSequence
  (params (EventMap :type Map%20of%20ESparksMidiTextEvents%20Enums%20to%20Stage%20Show%20Sequence%20Pool%20Object%20References) (Event :type ESparksMidiTextEvents%20Enum%20%28by%20ref%29))
  (bind (_value _returnvalue) (Utilities|Map|Find EventMap Event))
  ;[0.5601 1 0.3808] Pass the pool array to start a sequence
  (if _returnvalue
    (CallFunction|PlaySequence (Variables|/Script/SparksStageShowRuntime.StageShowSequencePool|GetSequences _value))
    (if (Variables|Default|GetEnableDebugLogs)
      (Development|PrintString (Utilities|String|Append "[StageManager] Playing MIDI event sequence: " (Utilities|String|EnumtoString Event)) true false "(R=0.000000,G=0.660000,B=1.000000,A=1.000000)" 3e+01 "StageManagerSequence"))))

(fn PlayQuickplayEventSequence
  (params (EventMap :type Map%20of%20EPilgrimQuickplayState%20Enums%20to%20Stage%20Show%20Sequence%20Pool%20Object%20References :default "Invalid") (State :type EPilgrimQuickplayState%20Enum :default "Intro"))
  (bind (_value _returnvalue) (Utilities|Map|Find EventMap State))
  ;[0.5601 1 0.3808] Pass the pool array to start a sequence
  (if _returnvalue
    (CallFunction|PlaySequence (Variables|/Script/SparksStageShowRuntime.StageShowSequencePool|GetSequences _value))
    (if (Variables|Default|GetEnableDebugLogs)
      (Development|PrintString (Utilities|String|Append "[StageManager] Playing gameplay state sequence: " (Utilities|String|EnumtoString State)) true false "(R=0.000000,G=0.660000,B=1.000000,A=1.000000)" 3e+01 "StageManagerSequence"))))

(fn SetCurrentColorPaletteAsGameplay
  (bind _currentgameplayeventcp (Variables|ColorShift|GetCurrentGameplayEventCP))
  (Utilities|IsValid _currentgameplayeventcp
    (:Is%20Valid
      (Variables|ColorShift|SetCurrentColorPalette (Variables|/PilgrimQuickplay/StageManager/StageManager/ColorPalette/BP_MojoStageColorPaletteGameplay_DataAsset.BP_MojoStageColorPaletteGameplay_DataAsset_C|GetColorPalette _currentgameplayeventcp))
      (Variables|ColorShift|SetCurrentAltColorPalette (Variables|/PilgrimQuickplay/StageManager/StageManager/ColorPalette/BP_MojoStageColorPaletteGameplay_DataAsset.BP_MojoStageColorPaletteGameplay_DataAsset_C|GetAltColorPalette _currentgameplayeventcp))
      ;[1 0.5357 0.2708] Debug Log
      (if (Variables|Default|GetEnableDebugLogs)
        (Development|PrintText (Utilities|Text|FormatText "NSLOCTEXT(\"[DC252DB94F2F9B61224EC987A5DD07B9]\", \"978519DA4D28EAED6A710AAF32724CAE\", \"[StageManager] Switched to Gameplay color palette: {name}\")" (Utilities|GetDisplayName (Variables|ColorShift|GetCurrentGameplayEventCP))) true true "(R=0.093866,G=1.000000,B=0.081323,A=1.000000)" 15.0 "CurrentColorPalette")))
    (:Is%20Not%20Valid
      (Fortnite|FortShippingLog self "[StageManager] Gameplay events color map is invalid. Gameplay events won't change color palette." true))))

(fn SetCurrentColorPaletteAsMIDI
  (bind _currentmidieventscp (Variables|ColorShift|GetCurrentMidiEventsCP))
  (Utilities|IsValid _currentmidieventscp
    (:Is%20Valid
      (Variables|ColorShift|SetCurrentColorPalette (Variables|/PilgrimQuickplay/StageManager/StageManager/ColorPalette/BP_MojoStageColorPaletteMidi_DataAsset.BP_MojoStageColorPaletteMidi_DataAsset_C|GetColorPalette _currentmidieventscp))
      (Variables|ColorShift|SetCurrentAltColorPalette (Variables|/PilgrimQuickplay/StageManager/StageManager/ColorPalette/BP_MojoStageColorPaletteMidi_DataAsset.BP_MojoStageColorPaletteMidi_DataAsset_C|GetAltColorPalette _currentmidieventscp))
      ;[1 0.5357 0.2708] Debug Log
      (if (Variables|Default|GetEnableDebugLogs)
        (Development|PrintText (Utilities|Text|FormatText "NSLOCTEXT(\"[DC252DB94F2F9B61224EC987A5DD07B9]\", \"1D2F440543FE86F18C8A09B3B66EE51D\", \"[StageManager] Switched to MIDI color palette: {name}\")" (Utilities|GetDisplayName (Variables|ColorShift|GetCurrentMidiEventsCP))) true true "(R=0.093866,G=1.000000,B=0.081323,A=1.000000)" 15.0 "CurrentColorPalette")))
    (:Is%20Not%20Valid
      (Fortnite|FortShippingLog self "[StageManager] MIDI events color map is invalid. MIDI events won't change color palette." true))))

(fn SetGameplayMaps
  (params (Sequence%20Map :type Stage%20Show%20Gameplay%20Event%20Map%20Object%20Reference%20%28by%20ref%29) (Color%20Palette%20Map :type BP%20Mojo%20Stage%20Color%20Palette%20Gameplay%20Data%20Asset%20Object%20Reference%20%28by%20ref%29))
  (Variables|Default|SetGameplayEventMap Sequence%20Map)
  (bind _output_get (Variables|Default|SetGameplayEventColorPalette Color%20Palette%20Map))
  (bind _output_get_2 (Variables|ColorShift|SetCurrentGameplayEventCP _output_get))
  (if (Variables|Default|GetEnableDebugLogs)
    (Development|PrintString (Utilities|String|Append "[StageManager] Gameplay map set: " (Utilities|GetDisplayName _output_get_2)))))

(fn SetMIDIMaps
  (params (Sequence%20Map :type Stage%20Show%20MIDIEvent%20Map%20Object%20Reference%20%28by%20ref%29) (Color%20Palette%20Map :type BP%20Mojo%20Stage%20Color%20Palette%20Midi%20Data%20Asset%20Object%20Reference%20%28by%20ref%29))
  (Utilities|IsValid Sequence%20Map
    (:Is%20Valid
      (Variables|Default|SetMIDIEventMap Sequence%20Map))
    (:Is%20Not%20Valid
      (Fortnite|FortShippingLog self "[StageManager] MIDI Sequence map not updated. May cause errors." true)))
  (Utilities|IsValid Color%20Palette%20Map
    (:Is%20Valid
      (bind _output_get (Variables|Default|SetMIDIEventColorPalette Color%20Palette%20Map))
      (Variables|ColorShift|SetCurrentMidiEventsCP _output_get))
    (:Is%20Not%20Valid
      (Fortnite|FortShippingLog self "[StageManager] Updating Color Palette failed. Invalid Color Palette Data Asset." true))))

(fn UpdateStageColors
  (bind (_synctype _frequency _eventcolorpalettes) (Utilities|Struct|BreakStageManagerEventColors (Variables|ColorShift|GetCurrentColorSettings)))
  (bind _currentcolorpalette (Variables|ColorShift|GetCurrentColorPalette))
  (bind _currentaltcolorpalette (Variables|ColorShift|GetCurrentAltColorPalette))
  (bind _result (ArrayMacros|SafeArrayGet _eventcolorpalettes (Variables|ColorShift|GetColorSettingsIndex))
    (:%20%20
      ;[1 1 1] Validate requested color palette indexes
      (if (Math|Boolean|ANDBoolean (< (Math|Conversions|ToInteger(Byte) (Utilities|Struct|BreakStageColor (Utilities|Struct|BreakStageManagerColorPaletteIndices _result))) (Utilities|Array|Length _currentcolorpalette)) (< (Math|Conversions|ToInteger(Byte) (Utilities|Struct|BreakStageColor (.Secondary (Utilities|Struct|BreakStageManagerColorPaletteIndices _result)))) (Utilities|Array|Length _currentcolorpalette)) (< (Math|Conversions|ToInteger(Byte) (Utilities|Struct|BreakStageColor (.Tertiary (Utilities|Struct|BreakStageManagerColorPaletteIndices _result)))) (Utilities|Array|Length _currentcolorpalette)) (< (Math|Conversions|ToInteger(Byte) (Utilities|Struct|BreakStageColor (.AltPrimary (Utilities|Struct|BreakStageManagerColorPaletteIndices _result)))) (Utilities|Array|Length _currentaltcolorpalette)) (< (Math|Conversions|ToInteger(Byte) (Utilities|Struct|BreakStageColor (.AltSecondary (Utilities|Struct|BreakStageManagerColorPaletteIndices _result)))) (Utilities|Array|Length _currentaltcolorpalette)) (< (Math|Conversions|ToInteger(Byte) (Utilities|Struct|BreakStageColor (.AltTertiary (Utilities|Struct|BreakStageManagerColorPaletteIndices _result)))) (Utilities|Array|Length _currentaltcolorpalette)))
        ;[1 1 1] Update Param Collection
        (Rendering|Material|SetVectorParameterValue "/SparksEffects/Effects/MPC_Sparks.MPC_Sparks" "StageShowPrimary" (Utilities|Array|Get(aref) _currentcolorpalette (Math|Conversions|ToInteger(Byte) (Utilities|Struct|BreakStageColor (Utilities|Struct|BreakStageManagerColorPaletteIndices _result)))) :!class /Script/Engine.KismetMaterialLibrary)
        (Rendering|Material|SetVectorParameterValue "/SparksEffects/Effects/MPC_Sparks.MPC_Sparks" "StageShowSecondary" (Utilities|Array|Get(aref) _currentcolorpalette (Math|Conversions|ToInteger(Byte) (Utilities|Struct|BreakStageColor (.Secondary (Utilities|Struct|BreakStageManagerColorPaletteIndices _result))))) :!class /Script/Engine.KismetMaterialLibrary)
        (Rendering|Material|SetVectorParameterValue "/SparksEffects/Effects/MPC_Sparks.MPC_Sparks" "StageShowTertiary" (Utilities|Array|Get(aref) _currentcolorpalette (Math|Conversions|ToInteger(Byte) (Utilities|Struct|BreakStageColor (.Tertiary (Utilities|Struct|BreakStageManagerColorPaletteIndices _result))))) :!class /Script/Engine.KismetMaterialLibrary)
        ;[1 1 1] Call for a Color update on all managed FxManagers
        (bind _values (Utilities|Map|Values (Variables|Default|GetFxManagersMap)))
        (for _array_element _values
          (ColorShift|CallColorShiftonActors _array_element (Utilities|Array|Get(aref) _currentcolorpalette (Math|Conversions|ToInteger(Byte) (Utilities|Struct|BreakStageColor (Utilities|Struct|BreakStageManagerColorPaletteIndices _result)))) (Utilities|Array|Get(aref) _currentcolorpalette (Math|Conversions|ToInteger(Byte) (Utilities|Struct|BreakStageColor (.Secondary (Utilities|Struct|BreakStageManagerColorPaletteIndices _result))))) (Utilities|Array|Get(aref) _currentcolorpalette (Math|Conversions|ToInteger(Byte) (Utilities|Struct|BreakStageColor (.Tertiary (Utilities|Struct|BreakStageManagerColorPaletteIndices _result))))) (Utilities|Array|Get(aref) _currentaltcolorpalette (Math|Conversions|ToInteger(Byte) (Utilities|Struct|BreakStageColor (.AltPrimary (Utilities|Struct|BreakStageManagerColorPaletteIndices _result))))) (Utilities|Array|Get(aref) _currentaltcolorpalette (Math|Conversions|ToInteger(Byte) (Utilities|Struct|BreakStageColor (.AltSecondary (Utilities|Struct|BreakStageManagerColorPaletteIndices _result))))) (Utilities|Array|Get(aref) _currentaltcolorpalette (Math|Conversions|ToInteger(Byte) (Utilities|Struct|BreakStageColor (.AltTertiary (Utilities|Struct|BreakStageManagerColorPaletteIndices _result)))))))
        ;[1 0.7269 0.4779] Debug log
        (if (Variables|Default|GetEnableDebugLogs)
          (Development|PrintText (Utilities|Text|FormatText "NSLOCTEXT(\"[DC252DB94F2F9B61224EC987A5DD07B9]\", \"68C7233F4C01B5699658BDA25C6181DF\", \"[StageManager] Primary Color ({index}): #{colorHex} \")" (Math|Conversions|ToInteger(Byte) (Utilities|Struct|BreakStageColor (Utilities|Struct|BreakStageManagerColorPaletteIndices _result))) (Math|Color|ToHex (Utilities|Array|Get(aref) _currentcolorpalette (Math|Conversions|ToInteger(Byte) (Utilities|Struct|BreakStageColor (Utilities|Struct|BreakStageManagerColorPaletteIndices _result)))))) true false (Utilities|Array|Get(aref) _currentcolorpalette (Math|Conversions|ToInteger(Byte) (Utilities|Struct|BreakStageColor (Utilities|Struct|BreakStageManagerColorPaletteIndices _result)))) 6e+01 "CPPrimary")
          (Development|PrintText (Utilities|Text|FormatText "NSLOCTEXT(\"[DC252DB94F2F9B61224EC987A5DD07B9]\", \"F0BC3610417E3CD0D16DE6B6A22E4AEF\", \"[StageManager] Secondary Color ({index}): #{colorHex} \")" (Math|Conversions|ToInteger(Byte) (Utilities|Struct|BreakStageColor (.Secondary (Utilities|Struct|BreakStageManagerColorPaletteIndices _result)))) (Math|Color|ToHex (Utilities|Array|Get(aref) _currentcolorpalette (Math|Conversions|ToInteger(Byte) (Utilities|Struct|BreakStageColor (.Secondary (Utilities|Struct|BreakStageManagerColorPaletteIndices _result))))))) true false (Utilities|Array|Get(aref) _currentcolorpalette (Math|Conversions|ToInteger(Byte) (Utilities|Struct|BreakStageColor (.Secondary (Utilities|Struct|BreakStageManagerColorPaletteIndices _result))))) 6e+01 "CPSecondary")
          (Development|PrintText (Utilities|Text|FormatText "NSLOCTEXT(\"[DC252DB94F2F9B61224EC987A5DD07B9]\", \"280C204344CABA832B58F08D26062A29\", \"[StageManager] Tertiary Color ({index}): #{colorHex} \")" (Math|Conversions|ToInteger(Byte) (Utilities|Struct|BreakStageColor (.Tertiary (Utilities|Struct|BreakStageManagerColorPaletteIndices _result)))) (Math|Color|ToHex (Utilities|Array|Get(aref) _currentcolorpalette (Math|Conversions|ToInteger(Byte) (Utilities|Struct|BreakStageColor (.Tertiary (Utilities|Struct|BreakStageManagerColorPaletteIndices _result))))))) true false (Utilities|Array|Get(aref) _currentcolorpalette (Math|Conversions|ToInteger(Byte) (Utilities|Struct|BreakStageColor (.Tertiary (Utilities|Struct|BreakStageManagerColorPaletteIndices _result))))) 6e+01 "CPTertiary")
          (Development|PrintText (Utilities|Text|FormatText "NSLOCTEXT(\"[DC252DB94F2F9B61224EC987A5DD07B9]\", \"98FB2FC0419EFEB4CCE68BA8596B2070\", \"[StageManager] Alt Primary Color ({index}): #{colorHex} \")" (Math|Conversions|ToInteger(Byte) (Utilities|Struct|BreakStageColor (.AltPrimary (Utilities|Struct|BreakStageManagerColorPaletteIndices _result)))) (Math|Color|ToHex (Utilities|Array|Get(aref) _currentaltcolorpalette (Math|Conversions|ToInteger(Byte) (Utilities|Struct|BreakStageColor (.AltPrimary (Utilities|Struct|BreakStageManagerColorPaletteIndices _result))))))) true false (Utilities|Array|Get(aref) _currentaltcolorpalette (Math|Conversions|ToInteger(Byte) (Utilities|Struct|BreakStageColor (.AltPrimary (Utilities|Struct|BreakStageManagerColorPaletteIndices _result))))) 6e+01 "AltCPPrimary")
          (Development|PrintText (Utilities|Text|FormatText "NSLOCTEXT(\"[DC252DB94F2F9B61224EC987A5DD07B9]\", \"4D857B1D44C2A60945DB15B5D4F2AD51\", \"[StageManager] Alt Secondary Color ({index}): #{colorHex} \")" (Math|Conversions|ToInteger(Byte) (Utilities|Struct|BreakStageColor (.AltSecondary (Utilities|Struct|BreakStageManagerColorPaletteIndices _result)))) (Math|Color|ToHex (Utilities|Array|Get(aref) _currentaltcolorpalette (Math|Conversions|ToInteger(Byte) (Utilities|Struct|BreakStageColor (.AltSecondary (Utilities|Struct|BreakStageManagerColorPaletteIndices _result))))))) true false (Utilities|Array|Get(aref) _currentaltcolorpalette (Math|Conversions|ToInteger(Byte) (Utilities|Struct|BreakStageColor (.AltSecondary (Utilities|Struct|BreakStageManagerColorPaletteIndices _result))))) 6e+01 "AltCPSecondary")
          (Development|PrintText (Utilities|Text|FormatText "NSLOCTEXT(\"[DC252DB94F2F9B61224EC987A5DD07B9]\", \"521B102D4DB987FDC3751FB1FDDF0516\", \"[StageManager] Alt Tertiary Color ({index}): #{colorHex} \")" (Math|Conversions|ToInteger(Byte) (Utilities|Struct|BreakStageColor (.AltTertiary (Utilities|Struct|BreakStageManagerColorPaletteIndices _result)))) (Math|Color|ToHex (Utilities|Array|Get(aref) _currentaltcolorpalette (Math|Conversions|ToInteger(Byte) (Utilities|Struct|BreakStageColor (.AltTertiary (Utilities|Struct|BreakStageManagerColorPaletteIndices _result))))))) true false (Utilities|Array|Get(aref) _currentaltcolorpalette (Math|Conversions|ToInteger(Byte) (Utilities|Struct|BreakStageColor (.AltTertiary (Utilities|Struct|BreakStageManagerColorPaletteIndices _result))))) 6e+01 "AltCPTertiary"))
        (else
          (Fortnite|FortShippingLog self "[StageManager] Color palette event settings points to a invalid color palette index. Skipping stage color update." true))))))

(fn UpdateColorsByGameplayEvent
  (params (State :type EPilgrimQuickplayState%20Enum%20%28by%20ref%29 :default "Invalid"))
  (bind (_color_settings _isvalid) (ColorShift|GetGameplayEventColorSettings State))
  (if _isvalid
    (ColorShift|ApplyColorSettings _color_settings)))

(fn UpdateColorsByMIDIEvent
  (params (Event :type ESparksMidiTextEvents%20Enum%20%28by%20ref%29))
  (bind (_color_settings _isvalid) (ColorShift|GetMIDIEventColorSettings Event))
  (if _isvalid
    (ColorShift|ApplyColorSettings _color_settings)))

(fn ConstructionScript)
